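# Physics asset for the male medium SpeedDialBattle Jacket body cosmetic, defining bodies and joint limits for its jacket hem and belts.
# Two-link jacket chains (left/right back-side, side and corner capsules, middle boxes) hang from spine_01, while three-link belt chains hang from pelvis.
# Pelvis, spine_01 and thigh capsules supply body volumes without constraints between them.
# Narrow swing and twist limits restrain the chains, with twist locked on middle and lower back-side joints.
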PhysicsAsset /BRCosmetics/Characters/Player/Male/Medium/Bodies/M_MED_SpeedDialBattle_Jacket/Meshes/M_MED_SpeedDialBattle_Jacket_Physics
bodies {
 belt_l [{shapeType=Capsule, length=2.1837, radius=1.6453, shapeName=belt_l_capsule, center=<1.1425 0.1139 -0.3474>, rotation=<rot 90 0 0>}]
 belt_l1 [{
   shapeType=Capsule
   length=2.5474
   radius=1.5647
   shapeName=belt_l1_capsule
   center=<0.4847 -0.2512 -0.4049>
   rotation=<rot 86.5081 179.9996 179.9995>
  }]
 belt_l2 [{
   shapeType=Capsule
   length=2.6038
   radius=1.8588
   shapeName=belt_l2_capsule
   center=<1.7106 -1.2832 -0.5779>
   rotation=<rot 62.1516 13.4905 16.435>
  }]
 belt_r [{shapeType=Capsule, length=2.3093, radius=1.1907, shapeName=belt_r_capsule, center=<1.5538 -0.0942 -0.4657>, rotation=<rot 77.1883 0 -0.0001>}]
 belt_r1 [{shapeType=Capsule, length=3.08, radius=1.7939, shapeName=belt_r1_capsule, center=<0.586 0.0253 -0.0793>, rotation=<rot 90 0 0>}]
 belt_r2 [{
   shapeType=Capsule
   length=3.319
   radius=1.7546
   shapeName=belt_r2_capsule
   center=<1.8956 -0.8062 0.7019>
   rotation=<rot 61.5548 -174.4382 -170.9285>
  }]
 dyn_Jacket_l_backSide_01_jnt [{shapeType=Capsule, length=5, radius=4, shapeName=dyn_Jacket_l_backSide_01_jnt_capsule, center=<0 -1.4994 1.5696>, rotation=<rot 90 0 0>}]
 dyn_Jacket_l_backSide_02_jnt [{shapeType=Capsule, length=5, radius=4, shapeName=dyn_Jacket_l_backSide_02_jnt_capsule, center=<3.3407 -1.4994 1.5696>, rotation=<rot 90 0 0>}]
 dyn_Jacket_l_corner_01_jnt [{
   shapeType=Capsule
   length=9.5215
   radius=2.7416
   shapeName=dyn_Jacket_l_corner_01_jnt_capsule
   center=<0 1.3671 1.2713>
   rotation=<rot 90 0 0>
  }]
 dyn_Jacket_l_corner_02_jnt [{
   shapeType=Capsule
   length=11.3463
   radius=2.6506
   shapeName=dyn_Jacket_l_corner_02_jnt_capsule
   center=<0 0.8814 0>
   rotation=<rot 84.5785 0 0>
  }]
 dyn_Jacket_l_side_01_jnt [{shapeType=Capsule, length=5.1481, radius=3.8519, shapeName=dyn_Jacket_l_side_01_jnt_capsule, center=<0 0.1871 1.5991>, rotation=<rot 90 0 0>}]
 dyn_Jacket_l_side_02_jnt [{shapeType=Capsule, length=5.1481, radius=3.8519, shapeName=dyn_Jacket_l_side_02_jnt_capsule, center=<0 0.1871 1.5991>, rotation=<rot 90 0 0>}]
 dyn_Jacket_m_01_jnt [{shapeType=Box, extentX=7.5, extentY=5.8765, extentZ=22.0641, shapeName=dyn_Jacket_m_01_jnt_box, center=<0 -2.9186 0>}]
 dyn_Jacket_m_02_jnt [{shapeType=Box, extentX=7.5, extentY=5.3072, extentZ=20.2656, shapeName=dyn_Jacket_m_02_jnt_box}]
 dyn_Jacket_r_backSide_01_jnt [{shapeType=Capsule, length=5, radius=4, shapeName=dyn_Jacket_l_backSide_01_jnt_capsule, center=<0 1.4994 -1.5696>, rotation=<rot 90 0 180>}]
 dyn_Jacket_r_backSide_02_jnt [{
   shapeType=Capsule
   length=5
   radius=4
   shapeName=dyn_Jacket_l_backSide_02_jnt_capsule
   center=<-3.3407 1.4994 -1.5696>
   rotation=<rot 90 0 180>
  }]
 dyn_Jacket_r_corner_01_jnt [{
   shapeType=Capsule
   length=9.0215
   radius=3.2416
   shapeName=dyn_Jacket_l_corner_01_jnt_capsule
   center=<0 -1.3671 -1.2713>
   rotation=<rot 90 0 180>
  }]
 dyn_Jacket_r_corner_02_jnt [{
   shapeType=Capsule
   length=11.3463
   radius=2.6506
   shapeName=dyn_Jacket_l_corner_02_jnt_capsule
   center=<0 -3.165 -0.0136>
   rotation=<rot 84.5785 0 180>
  }]
 dyn_Jacket_r_side_01_jnt [{
   shapeType=Capsule
   length=5.1481
   radius=3.8519
   shapeName=dyn_Jacket_l_side_01_jnt_capsule
   center=<0 -0.1871 -1.5991>
   rotation=<rot 90 0 180>
  }]
 dyn_Jacket_r_side_02_jnt [{
   shapeType=Capsule
   length=5.1481
   radius=3.8519
   shapeName=dyn_Jacket_l_side_02_jnt_capsule
   center=<0 -0.1871 -1.5991>
   rotation=<rot 90 0 180>
  }]
 pelvis [
  {shapeType=Capsule, length=5.3311, radius=10.7615, shapeName=pelvis_capsule, center=<1.108 0.7676 -0.011>, rotation=<rot 5.1404 -90 -0.0505>}
  {shapeType=Capsule, length=10.6399, radius=8.619, shapeName=pelvis_capsule, center=<5.8732 -5.8427 0>}
 ]
 spine_01 [
  {
   shapeType=Capsule
   length=0.0405
   radius=16.2451
   shapeName=spine_01_capsule
   center=<7.7213 -0.5366 0>
   rotation=<rot -90 -179.9999 179.9996>
  }
  {shapeType=Capsule, length=4.164, radius=11.11, shapeName=spine_01_capsule, center=<-1.8495 1.0237 0>, rotation=<rot 0 179.9997 -180>}
 ]
 thigh_l [{
   shapeType=Capsule
   length=20.2314
   radius=8.1159
   shapeName=thigh_l_capsule
   center=<-9.3834 -0.1983 0.293>
   rotation=<rot 14.5199 -88.8908 -97.604>
  }]
 thigh_r [{
   shapeType=Capsule
   length=20.4719
   radius=7.8755
   shapeName=thigh_l_capsule
   center=<8.9082 -2.2611 0.3597>
   rotation=<rot 14.6598 -92.6137 90.7837>
  }]
}
constraints [
 {bone1Name=belt_l1, bone2Name=belt_l, swing1LimitDegrees=25, swing2LimitDegrees=25, twistLimitDegrees=5}
 {bone1Name=belt_l2, bone2Name=belt_l1, swing1LimitDegrees=25, swing2LimitDegrees=15, twistLimitDegrees=5}
 {bone1Name=belt_r1, bone2Name=belt_r, swing1LimitDegrees=10, swing2LimitDegrees=6, twistLimitDegrees=5}
 {bone1Name=belt_r2, bone2Name=belt_r1, swing1LimitDegrees=15, swing2LimitDegrees=7, twistLimitDegrees=5}
 {bone1Name=belt_l, bone2Name=pelvis, swing1LimitDegrees=15, swing2LimitDegrees=15, twistLimitDegrees=1}
 {bone1Name=belt_r, bone2Name=pelvis, swing1LimitDegrees=5, swing2LimitDegrees=3, twistLimitDegrees=2}
 {bone1Name=dyn_Jacket_l_backSide_01_jnt, bone2Name=spine_01, swing1LimitDegrees=17.5, swing2LimitDegrees=3.2, twistLimitDegrees=1}
 {
  bone1Name=dyn_Jacket_l_backSide_02_jnt
  bone2Name=dyn_Jacket_l_backSide_01_jnt
  swing1LimitDegrees=10
  swing2LimitDegrees=6
  twistLimitDegrees=1
  twistMotion=Locked
 }
 {bone1Name=dyn_Jacket_l_corner_01_jnt, bone2Name=spine_01, swing1LimitDegrees=5, swing2LimitDegrees=14.5, twistLimitDegrees=1}
 {
  bone1Name=dyn_Jacket_l_corner_02_jnt
  bone2Name=dyn_Jacket_l_corner_01_jnt
  swing1LimitDegrees=15
  swing2LimitDegrees=15
  twistLimitDegrees=1
 }
 {bone1Name=dyn_Jacket_l_side_01_jnt, bone2Name=spine_01, swing1LimitDegrees=4, swing2LimitDegrees=21.52, twistLimitDegrees=1}
 {bone1Name=dyn_Jacket_l_side_02_jnt, bone2Name=dyn_Jacket_l_side_01_jnt, swing1LimitDegrees=3, swing2LimitDegrees=7, twistLimitDegrees=1}
 {bone1Name=dyn_Jacket_r_backSide_01_jnt, bone2Name=spine_01, swing1LimitDegrees=17.5, swing2LimitDegrees=3.2, twistLimitDegrees=1}
 {
  bone1Name=dyn_Jacket_r_backSide_02_jnt
  bone2Name=dyn_Jacket_r_backSide_01_jnt
  swing1LimitDegrees=10
  swing2LimitDegrees=6
  twistLimitDegrees=1
  twistMotion=Locked
 }
 {bone1Name=dyn_Jacket_m_01_jnt, bone2Name=spine_01, swing1LimitDegrees=18, swing2LimitDegrees=2, twistLimitDegrees=1, twistMotion=Locked}
 {
  bone1Name=dyn_Jacket_m_02_jnt
  bone2Name=dyn_Jacket_m_01_jnt
  swing1LimitDegrees=14
  swing2LimitDegrees=2
  twistLimitDegrees=1
  twistMotion=Locked
 }
 {bone1Name=dyn_Jacket_r_corner_02_jnt, bone2Name=dyn_Jacket_r_corner_01_jnt, swing1LimitDegrees=5, swing2LimitDegrees=15, twistLimitDegrees=1}
 {bone1Name=dyn_Jacket_r_corner_01_jnt, bone2Name=spine_01, swing1LimitDegrees=5, swing2LimitDegrees=14.5, twistLimitDegrees=1}
 {bone1Name=dyn_Jacket_r_side_01_jnt, bone2Name=spine_01, swing1LimitDegrees=4, swing2LimitDegrees=21.52, twistLimitDegrees=1}
 {bone1Name=dyn_Jacket_r_side_02_jnt, bone2Name=dyn_Jacket_r_side_01_jnt, swing1LimitDegrees=3, swing2LimitDegrees=7, twistLimitDegrees=1}
]
