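# Physics asset for the dangling necklace on the SpeedDialBattle Jacket male medium cosmetic body.
# A four-joint chain of three small capsules and an end box, likely the pendant, hangs from spine_05.
# Several overlapping capsules on spine_05 approximate the upper torso as collision volume for the chain.
# Each link constrains to the one above with locked twist and swing limits that widen toward the tip, keeping the root stiff and the end looser.
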PhysicsAsset /BRCosmetics/Characters/Player/Male/Medium/Bodies/M_MED_SpeedDialBattle_Jacket/Meshes/M_MED_SpeedDialBattle_Jacket_Physics_necklace
bodies {
 dyn_necklace_jnt1 [{shapeType=Capsule, length=2.7234, radius=1.8548, shapeName=dyn_necklace_jnt1_capsule, center=<0 0.2739 0>, rotation=<rot 90 0 0>}]
 dyn_necklace_jnt2 [{shapeType=Capsule, length=3.5194, radius=1.4806, shapeName=dyn_necklace_jnt2_capsule, rotation=<rot 90 0 0>}]
 dyn_necklace_jnt3 [{shapeType=Capsule, length=1.3613, radius=0.9903, shapeName=dyn_necklace_jnt3_capsule, rotation=<rot 90 0 -11.868>}]
 dyn_necklace_jnt4 [{
   shapeType=Box
   extentX=1.6014
   extentY=6.8213
   extentZ=6.4433
   shapeName=dyn_necklace_jnt4_box
   center=<3.8396 0.6935 -0.2975>
   rotation=<rot 0 97.9572 0>
  }]
 spine_05 [
  {shapeType=Capsule, length=15.4016, radius=7.5739, shapeName=spine_05_capsule, center=<-7.0715 7.3563 -0.1936>}
  {shapeType=Capsule, length=15.3456, radius=9.967, shapeName=spine_05_capsule, center=<-2.7906 1.7536 0>}
  {
   shapeType=Capsule
   length=11.3152
   radius=3.3034
   shapeName=spine_05_capsule
   center=<0.876 12.7153 6.4289>
   rotation=<rot -78.5818 173.2572 155.2331>
  }
  {
   shapeType=Capsule
   length=11.4003
   radius=3.5997
   shapeName=spine_05_capsule
   center=<0.7018 12.8501 -6.8748>
   rotation=<rot -84.7032 -96.3627 66.5123>
  }
  {
   shapeType=Capsule
   length=8.3461
   radius=11.3155
   shapeName=spine_05_capsule
   center=<4.4338 -5.6221 -0.1503>
   rotation=<rot -1.4365 -0.0643 0>
  }
  {
   shapeType=Capsule
   length=7.5849
   radius=3.9848
   shapeName=spine_05_capsule
   center=<-2.2514 8.7895 -0.0595>
   rotation=<rot 0.7979 -44.6604 1.339>
  }
 ]
}
constraints [
 {bone1Name=dyn_necklace_jnt1, bone2Name=spine_05, swing1LimitDegrees=4, swing2LimitDegrees=15, twistLimitDegrees=45, twistMotion=Locked}
 {
  bone1Name=dyn_necklace_jnt2
  bone2Name=dyn_necklace_jnt1
  swing1LimitDegrees=15
  swing2LimitDegrees=20
  twistLimitDegrees=45
  twistMotion=Locked
 }
 {
  bone1Name=dyn_necklace_jnt3
  bone2Name=dyn_necklace_jnt2
  swing1LimitDegrees=20
  swing2LimitDegrees=25
  twistLimitDegrees=45
  twistMotion=Locked
 }
 {
  bone1Name=dyn_necklace_jnt4
  bone2Name=dyn_necklace_jnt3
  swing1LimitDegrees=45
  swing2LimitDegrees=45
  twistLimitDegrees=45
  twistMotion=Locked
 }
]
