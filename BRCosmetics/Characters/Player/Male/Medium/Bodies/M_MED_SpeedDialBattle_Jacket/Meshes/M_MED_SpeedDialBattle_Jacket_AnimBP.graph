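# Animation Blueprint for the SpeedDialBattle jacket on medium male player bodies, adding secondary motion and pose correctives to the pose copied from the parent mesh.
# After melee-dependent neck scaling, two RigidBody passes simulate jacket and necklace, feeding a crouch-enabled Control Rig and a deform rig.
# Additive bone transforms then correct collar, belts and jacket corners.
# The EventGraph drives these using gravity override, head and spine swing angles, and crouch, movement, melee and targeting states.
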AnimBlueprint /BRCosmetics/Characters/Player/Male/Medium/Bodies/M_MED_SpeedDialBattle_Jacket/Meshes/M_MED_SpeedDialBattle_Jacket_AnimBP
target_skeleton=&/M_MED_SpeedDialBattle_Jacket_Skeleton
graphs [
 AnimGraph {nodes [
   Output Pose(Result=Use cached pose 'when in the Crouch cor.Jacket '()) {layerGroup=None}
   n1 = "Transform (Modify) Bone - Bone: correct_neck_jnt"!(
    ComponentPose=Local To Component(LocalPose=Copy Pose From Mesh() {bUseAttachedParent=true})
    bAlphaBoolEnabled:bool=Get is_not_meleeAnim() -> (is_not_meleeAnim:bool)
    Translation="0,0,0"
    Rotation="0,0,0"
    Scale="1.4,1.4,1.4"
   ) -> (Pose) {
    alphaInputType=Bool
    rotationMode=BMM_Additive
    rotationSpace=BCS_BoneSpace
    scaleMode=BMM_Additive
    scaleSpace=BCS_BoneSpace
    translationMode=BMM_Additive
    translationSpace=BCS_BoneSpace
    alphaBoolBlend {blendInTime=0.15, blendOption=Linear, blendOutTime=0.15}
    boneToModify {boneName=correct_neck_jnt}
    scale=<1.4 1.4 1.4>
   }
   n2 = "Transform (Modify) Bone - Bone: correct_neck_jnt"(ComponentPose=@n1, bAlphaBoolEnabled=Get bIsPlayingMeleeAnim(), Translation="0,0,0", Rotation="0,0,0", Scale="2,2,2") {
    alphaInputType=Bool
    rotationMode=BMM_Additive
    rotationSpace=BCS_BoneSpace
    scaleMode=BMM_Additive
    scaleSpace=BCS_BoneSpace
    translationMode=BMM_Additive
    translationSpace=BCS_BoneSpace
    alphaBoolBlend {blendInTime=0.15, blendOption=Linear, blendOutTime=0.15}
    boneToModify {boneName=correct_neck_jnt}
    scale=<2 2 2>
   }
   Save cached pose 'pre_dyn_cache'(Pose=Component To Local(ComponentPose=@n2)) {cacheName=pre_dyn_cache}
   n3 = RigidBody!(
    ComponentPose=Local To Component(LocalPose=Use cached pose 'pre_dyn_cache'())
    bAlphaBoolEnabled:bool
    OverrideWorldGravity:Vector=Get Gravity Override()
    ExternalForce
    bTransferBoneVelocities:bool=False
   ) -> (Pose) {
    alphaInputType=Bool
    bOverrideWorldGravity=true
    overridePhysicsAsset=&/M_MED_SpeedDialBattle_Jacket_Physics
    simulationSpace=BaseBoneSpace
    alphaBoolBlend {blendInTime=0.13, blendOption=Linear, blendOutTime=0.13}
    baseBoneRef {boneName=pelvis}
    componentAppliedLinearAccClamp=<1000 1000 1000>
    componentLinearVelScale=<0.3 0.3 0.3>
    simSpaceSettings {maxAngularAcceleration=70, maxAngularVelocity=1.5, maxLinearAcceleration=150, maxLinearVelocity=100, worldAlpha=1}
   }
   n4 = RigidBody!(ComponentPose=@n3, bAlphaBoolEnabled, ExternalForce, bTransferBoneVelocities=False) -> (Pose) {
    alphaInputType=Bool
    bOverrideWorldGravity=true
    overridePhysicsAsset=&/M_MED_SpeedDialBattle_Jacket_Physics_necklace
    simulationSpace=BaseBoneSpace
    alphaBoolBlend {blendInTime=0.13, blendOption=Linear, blendOutTime=0.13}
    baseBoneRef {boneName=spine_01}
    componentAppliedLinearAccClamp=<1000 1000 1000>
    componentLinearAccScale=<0.001 0.001 0.001>
    componentLinearVelScale=<0.001 0.001 0.001>
    overrideWorldGravity {z=-350}
    simSpaceSettings {maxAngularAcceleration=1000, maxAngularVelocity=3, maxLinearAcceleration=1350, maxLinearVelocity=15, velocityScaleZ=0.1}
   }
   n5 = Control Rig!(
    Source=Component To Local(ComponentPose=@n4)
    bAlphaBoolEnabled:bool=Get bIsCrouching()
    state_crouching_cond:bool=Get bIsMoving2D()
    if targeting:bool=Get cond_jacket_when aiming() -> (cond_jacket_when aiming:bool)
   ) -> (Pose) {
    alphaInputType=Bool
    alphaBoolBlend {blendInTime=0.1, blendOption=Linear, blendOutTime=0.5}
    controlRigAssetReference {blueprintRigClass=&/M_MED_SpeedDialBattle_Jacket_CtrlRig.M_MED_SpeedDialBattle_Jacket_CtrlRig_C}
   }
   n6 = Control Rig!(Source=@n5, Alpha:float) -> (Pose) {
    alphaBoolBlend {blendOption=Linear}
    controlRigAssetReference {blueprintRigClass=&/M_MED_SpeedDialBattle_Jacket_DeformRig.M_MED_SpeedDialBattle_Jacket_DeformRig_C}
   }
   Save cached pose 'Dyn_cache'(Pose=@n6) {cacheName=Dyn_cache}
   n7 = Get bIsCrouching()
   n8 = "Transform (Modify) Bone - Bone: correctJoint_for_jacket_l"(ComponentPose=Local To Component(LocalPose=Use cached pose 'Dyn_cache'()), Alpha=1, Translation="0,0,0", Rotation="0,0,0", Scale="0.58,1,1") {
    scaleMode=BMM_Additive
    scaleSpace=BCS_BoneSpace
    alphaBoolBlend {blendOption=Linear}
    boneToModify {boneName=correctJoint_for_jacket_l}
    scale {x=0.58}
   }
   n9 = "Transform (Modify) Bone - Bone: correctJoint_for_jacket_r"(ComponentPose=@n8, Alpha=1, Translation="0,0,0", Rotation="0,0,0", Scale="0.58,1,1") {
    rotationMode=BMM_Additive
    rotationSpace=BCS_BoneSpace
    scaleMode=BMM_Additive
    scaleSpace=BCS_BoneSpace
    translationMode=BMM_Additive
    translationSpace=BCS_BoneSpace
    alphaBoolBlend {blendOption=Linear}
    boneToModify {boneName=correctJoint_for_jacket_r}
    scale {x=0.58}
   }
   n10 = "Transform (Modify) Bone - Bone: correct_collar_joint"(
    ComponentPose=@n9
    Alpha=Get Out Swing Angle_Head() -> (Out Swing Angle_Head:float)
    Translation="0.365,-2.007,-12.905"
    Rotation="-0.142,0.132,-68.171"
    Scale="1,1,1"
   ) {
    rotationMode=BMM_Additive
    rotationSpace=BCS_BoneSpace
    scaleMode=BMM_Additive
    scaleSpace=BCS_BoneSpace
    translationMode=BMM_Additive
    translationSpace=BCS_BoneSpace
    alphaBoolBlend {blendOption=Linear}
    alphaScaleBiasClamp {bInterpResult=true, scale=0.5}
    boneToModify {boneName=correct_collar_joint}
    rotation=<rot -0.142 0.132 -68.171>
    translation=<0.365 -2.007 -12.905>
   }
   n11 = "Transform (Modify) Bone - Bone: belt_l1"(ComponentPose=@n10, Alpha=1, Translation="-0.234,0.906,-0.655", Rotation="-8.289,7.562,3.123", Scale="1,1,1") {
    rotationMode=BMM_Additive
    rotationSpace=BCS_BoneSpace
    scaleMode=BMM_Additive
    scaleSpace=BCS_BoneSpace
    translationMode=BMM_Additive
    translationSpace=BCS_BoneSpace
    alphaBoolBlend {blendOption=Linear}
    boneToModify {boneName=belt_l1}
    rotation=<rot -8.289 7.562 3.123>
    translation=<-0.234 0.906 -0.655>
   }
   n12 = "Transform (Modify) Bone - Bone: Belts_l_offset_jnt"(ComponentPose=@n11, Alpha=1, Translation="-0.308,0.000,-0.947", Rotation="-17.170,0.015,-0.067", Scale="1,1,1") {
    rotationMode=BMM_Additive
    rotationSpace=BCS_BoneSpace
    scaleMode=BMM_Additive
    scaleSpace=BCS_BoneSpace
    translationMode=BMM_Additive
    translationSpace=BCS_BoneSpace
    alphaBoolBlend {blendOption=Linear}
    boneToModify {boneName=Belts_l_offset_jnt}
    rotation=<rot -17.17 0.015 -0.067>
    translation {x=-0.308, z=-0.947}
   }
   n13 = "Transform (Modify) Bone - Bone: Belts_r_offset_jnt"(ComponentPose=@n12, Alpha=1, Translation="-0.052,-0.002,1.973", Rotation="-9.277,0.002,0.032", Scale="1,1,1") {
    rotationMode=BMM_Additive
    rotationSpace=BCS_BoneSpace
    scaleMode=BMM_Additive
    scaleSpace=BCS_BoneSpace
    translationMode=BMM_Additive
    translationSpace=BCS_BoneSpace
    alphaBoolBlend {blendOption=Linear}
    boneToModify {boneName=Belts_r_offset_jnt}
    rotation=<rot -9.277 0.002 0.032>
    translation=<-0.052 -0.002 1.973>
   }
   n14 = "Transform (Modify) Bone - Bone: dyn_Jacket_l_corner_02_jnt"(ComponentPose=@n13, Alpha=1, Translation="0,0,0", Rotation="20.907,22.882,-23.284", Scale="1,1,1") {
    rotationMode=BMM_Additive
    rotationSpace=BCS_BoneSpace
    scaleMode=BMM_Additive
    scaleSpace=BCS_BoneSpace
    translationMode=BMM_Additive
    translationSpace=BCS_BoneSpace
    alphaBoolBlend {blendOption=Linear}
    boneToModify {boneName=dyn_Jacket_l_corner_02_jnt}
    rotation=<rot 20.907 22.882 -23.284>
   }
   n15 = "Transform (Modify) Bone - Bone: belt_r"!(ComponentPose=@n14, bAlphaBoolEnabled:bool=@n7, Translation="0,0,0", Rotation="-7.594,4.414,6.704", Scale="1,1,1") -> (Pose) {
    alphaInputType=Bool
    rotationMode=BMM_Additive
    rotationSpace=BCS_BoneSpace
    scaleMode=BMM_Additive
    scaleSpace=BCS_BoneSpace
    translationMode=BMM_Additive
    translationSpace=BCS_BoneSpace
    alphaBoolBlend {blendInTime=0.13, blendOption=Linear, blendOutTime=0.13}
    boneToModify {boneName=belt_r}
    rotation=<rot -7.594 4.414 6.704>
   }
   n16 = "Transform (Modify) Bone - Bone: belt_r2"!(ComponentPose=@n15, bAlphaBoolEnabled:bool=@n7, Translation="0,0,0", Rotation="15.679,7.241,2.786", Scale="1,1,1") -> (Pose) {
    alphaInputType=Bool
    rotationMode=BMM_Additive
    rotationSpace=BCS_BoneSpace
    scaleMode=BMM_Additive
    scaleSpace=BCS_BoneSpace
    translationMode=BMM_Additive
    translationSpace=BCS_BoneSpace
    alphaBoolBlend {blendInTime=0.13, blendOption=Linear, blendOutTime=0.13}
    boneToModify {boneName=belt_r2}
    rotation=<rot 15.679 7.241 2.786>
   }
   n17 = Get correct_jacket_f_var_value() -> (correct_jacket_f_var_value:double)
   n18 = "Transform (Modify) Bone - Bone: dyn_Jacket_l_corner_01_jnt"(ComponentPose=@n16, Alpha=@n17, Translation="-0.990,-2.061,12.675", Rotation="-17.222,-33.042,-36.713", Scale="1,0.8,0.8") {
    rotationMode=BMM_Additive
    rotationSpace=BCS_BoneSpace
    scaleMode=BMM_Additive
    scaleSpace=BCS_BoneSpace
    translationMode=BMM_Additive
    translationSpace=BCS_BoneSpace
    alphaBoolBlend {blendOption=Linear}
    alphaScaleBiasClamp {bInterpResult=true, interpSpeedDecreasing=7}
    boneToModify {boneName=dyn_Jacket_l_corner_01_jnt}
    rotation=<rot -17.222 -33.042 -36.713>
    scale {y=0.8, z=0.8}
    translation=<-0.99 -2.061 12.675>
   }
   n19 = "Transform (Modify) Bone - Bone: dyn_Jacket_l_corner_02_jnt"(ComponentPose=@n18, Alpha=@n17, Translation="3.417,-14.116,2.369", Rotation="-45.839,4.108,-48.958", Scale="0.8,0.8,0.838") {
    rotationMode=BMM_Additive
    rotationSpace=BCS_BoneSpace
    scaleMode=BMM_Additive
    scaleSpace=BCS_BoneSpace
    translationMode=BMM_Additive
    translationSpace=BCS_BoneSpace
    alphaBoolBlend {blendOption=Linear}
    alphaScaleBiasClamp {bInterpResult=true, interpSpeedDecreasing=7}
    boneToModify {boneName=dyn_Jacket_l_corner_02_jnt}
    rotation=<rot -45.839 4.108 -48.958>
    scale=<0.8 0.8 0.838>
    translation=<3.417 -14.116 2.369>
   }
   n20 = "Transform (Modify) Bone - Bone: dyn_Jacket_l_side_01_jnt"(ComponentPose=@n19, Alpha=@n17, Translation="0.063,0.098,0.038", Rotation="15.448,-33.451,-27.061", Scale="0.757,0.631,1.000") {
    rotationMode=BMM_Additive
    rotationSpace=BCS_BoneSpace
    scaleMode=BMM_Additive
    scaleSpace=BCS_BoneSpace
    translationMode=BMM_Additive
    translationSpace=BCS_BoneSpace
    alphaBoolBlend {blendOption=Linear}
    alphaScaleBiasClamp {bInterpResult=true, interpSpeedDecreasing=7}
    boneToModify {boneName=dyn_Jacket_l_side_01_jnt}
    rotation=<rot 15.448 -33.451 -27.061>
    scale=<0.7569 0.631>
    translation=<0.063 0.098 0.038>
   }
   n21 = "Transform (Modify) Bone - Bone: belt_l"!(ComponentPose=@n20, bAlphaBoolEnabled:bool=@n7, Translation="0.171,0.254,0.433", Rotation="4.384,-35.491,35.418", Scale="1,1,1") -> (Pose) {
    alphaInputType=Bool
    rotationMode=BMM_Additive
    rotationSpace=BCS_BoneSpace
    scaleMode=BMM_Additive
    scaleSpace=BCS_BoneSpace
    translationMode=BMM_Additive
    translationSpace=BCS_BoneSpace
    alphaBoolBlend {blendInTime=0.13, blendOption=Linear, blendOutTime=0.13}
    boneToModify {boneName=belt_l}
    rotation=<rot 4.384 -35.491 35.418>
    translation=<0.171 0.254 0.433>
   }
   n22 = "Transform (Modify) Bone - Bone: belt_l2"!(ComponentPose=@n21, bAlphaBoolEnabled:bool=@n7, Translation="0.905,-1.403,-0.214", Rotation="0.362,3.829,-7.533", Scale="1,1,1") -> (Pose) {
    alphaInputType=Bool
    rotationMode=BMM_Additive
    rotationSpace=BCS_BoneSpace
    scaleMode=BMM_Additive
    scaleSpace=BCS_BoneSpace
    translationMode=BMM_Additive
    translationSpace=BCS_BoneSpace
    alphaBoolBlend {blendInTime=0.13, blendOption=Linear, blendOutTime=0.13}
    boneToModify {boneName=belt_l2}
    rotation=<rot 0.362 3.829 -7.533>
    translation=<0.905 -1.403 -0.214>
   }
   n23 = "Transform (Modify) Bone - Bone: defrom_extra_l"!(
    ComponentPose=@n22
    bAlphaBoolEnabled:bool=@n7
    Translation="-0.324,-4.761,4.271"
    Rotation="-0.009,-0.005,18.729"
    Scale="1.000,1.000,2.000"
   ) -> (Pose) {
    alphaInputType=Bool
    rotationMode=BMM_Additive
    rotationSpace=BCS_BoneSpace
    scaleMode=BMM_Additive
    scaleSpace=BCS_BoneSpace
    translationMode=BMM_Additive
    translationSpace=BCS_BoneSpace
    alphaBoolBlend {blendInTime=0.13, blendOption=Linear, blendOutTime=0.13}
    boneToModify {boneName=defrom_extra_l}
    rotation=<rot -0.009 -0.005 18.729>
    scale {z=2}
    translation=<-0.324 -4.761 4.271>
   }
   n24 = "Transform (Modify) Bone - Bone: belt_l"(
    ComponentPose=@n23
    bAlphaBoolEnabled=Get bl_correct_left_belt() -> (bl_correct_left_belt:bool)
    Translation="-0.086,2.139,0.253"
    Rotation="-1.513,3.735,-9.930"
    Scale="1,1,1"
   ) {
    alphaInputType=Bool
    rotationMode=BMM_Additive
    rotationSpace=BCS_BoneSpace
    scaleMode=BMM_Additive
    scaleSpace=BCS_BoneSpace
    translationMode=BMM_Additive
    translationSpace=BCS_BoneSpace
    alphaBoolBlend {blendInTime=0.05, blendOption=Linear, blendOutTime=0.13}
    boneToModify {boneName=belt_l}
    rotation=<rot -1.513 3.735 -9.93>
    translation=<-0.086 2.139 0.253>
   }
   Save cached pose 'when in the Crouch cor.Jacket '(Pose=Component To Local(ComponentPose=@n24)) {cacheName="when in the Crouch cor.Jacket "}
  ]}
 EventGraph {nodes [
   Event BlueprintUpdateAnimation()
   > n1 = Sequence!(execute) -> (then_0, then_1, then_2:Exec, then_3:Exec)
   Set sitting_or_standing_state(execute=@n1.then_1, sitting_or_standing_state:bool=OR Boolean(A=Get bIsCrouching(), B=Get bIsJumping())) -> (Output_Get:bool)
   n2 = GetOwnerSkeletalMeshComponent()
   Set Out Swing Angle_Head(
    execute=@n1.then_2
    Out Swing Angle_Head:float=MapRangeClamped(Value=GetTwistAndSwingAngleOfDeltaRotationFromRefPose(self=@n2, BoneName=head).OutSwingAngle, InRangeA=18, InRangeB=40)
   ) -> (Output_Get:float)
   Set is_not_meleeAnim(execute=@n1.then_3, is_not_meleeAnim:bool="Not Equal (Boolean)"(A=Get bIsPlayingMeleeAnim(), B=true) -> (ReturnValue=~)) -> (Output_Get:bool)
   n3 = Gravity Override State Function DEPRECATED(
    execute=@n1.then_0
    CharacterPartAnimInstance=Self-Reference()
    GameTimeInSeconds=GetGameTimeInSeconds()
    GravityOverrideParams=Get Gravity Override Params()
   )
   > Set Gravity Override(Gravity Override=@n3)
   > n4 = Sequence!(execute) -> (then_0, then_1, then_2)
   n5 = Get bIsMoving2D()
   n6 = "Not Equal (Boolean)"(A=@n5, B=true) -> (ReturnValue=~)
   n7 = Get bIsCrouching()
   n8 = "Not Equal (Boolean)"(A=Get bIsPlayingMeleeAnim(), B=true) -> (ReturnValue=~)
   n9 = AND Boolean(A=@n8, B=AND Boolean(A=@n7, B=@n6))
   n10 = MapRangeClamped(
    Value=GetTwistAndSwingAngleOfDeltaRotationFromRefPose(self=GetOwnerSkeletalMeshComponent(), BoneName=spine_04).OutSwingAngle
    InRangeA=6.5
    InRangeB=9.9
   )
   Set cond_jacket_bl_var(execute=@n4.then_1, cond_jacket_bl_var:bool=AND Boolean(A="float > float"(A=@n10), B=@n9)) -> (Output_Get:bool)
   > Set bl_correct_left_belt(bl_correct_left_belt:bool=AND Boolean(A=@n7, B=@n5)) -> (Output_Get:bool)
   n11 = Branch(execute=@n4.then_0, Condition=Get cond_jacket_bl_var() -> (cond_jacket_bl_var:bool))
   > Set correct_jacket_f_var_value(correct_jacket_f_var_value:double=@n10) -> (Output_Get:double)
   Set correct_jacket_f_var_value(execute=@n11.else, correct_jacket_f_var_value=0)
   Set cond_jacket_when aiming(execute=@n4.then_2, cond_jacket_when aiming:bool=AND Boolean!(A=Get bIsTargeting(), B=@n6, C:bool=@n8) -> (ReturnValue)) -> (Output_Get:bool)
   AND Boolean!(A=@n9, B=false) -> (ReturnValue)
  ]}
]
variables [
 Anim int {type=FortPlayerAnimInstance, default=null}
 bl_correct_left_belt {type=boolean}
 cond_jacket_bl_var {type=boolean}
 cond_jacket_when aiming {type=boolean}
 cond_kacjet_var {type=boolean}
 correct_jacket_f_var_min_value {type=number}
 correct_jacket_f_var_value {type=number}
 Gravity Override {type=Vector}
 Gravity Override Params {type=GravityOverrideParamsStruct}
 is_not_meleeAnim {type=boolean}
 Out Swing Angle_Head {type=number}
 sitting_or_standing_state {type=boolean}
 Skel_mesh {type=SkeletalMeshComponent}
]
